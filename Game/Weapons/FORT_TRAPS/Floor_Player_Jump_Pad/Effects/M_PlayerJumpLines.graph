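# Unlit translucent mesh-particle material for the line effects of the floor player jump pad trap.
# Particle color RGB feeds an emissive adjustment function, so the particle system drives the color.
# Particle alpha is multiplied by a depth fade and a camera distance blend with tunable range and start offset.
# A shading path switch reduces the mobile version to plain particle alpha, skipping both fades.
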Material /Game/Weapons/FORT_TRAPS/Floor_Player_Jump_Pad/Effects/M_PlayerJumpLines
graph {
 properties {bUsedWithMeshParticles=true, blendMode=BLEND_Translucent, pixelDepthOffsetMode=PDOM_Legacy, shadingModel=MSM_Unlit, usedShadingModels=Unlit}
 nodes [
  n1 = ParticleColor()
  MaterialFunctionCall(Emissive=@n1.RGB) {materialFunction=&MF_FX_EmissiveAdjustments}
  n2 = DepthFade()
  n3 = ScalarParameter() {defaultValue=256, parameterName=Blend Range}
  n4 = ScalarParameter() {defaultValue=-384, parameterName=Start Offset}
  ShadingPathSwitch(
   Default=Multiply(
    A=Multiply(A=@n1.A, B=@n2)
    B=MaterialFunctionCall!(Blend Range=@n3, Start Offset=@n4) {materialFunction=/Engine/Functions/Engine_MaterialFunctions03/Particles/Distance_Blend}.Result
   )
   Mobile=@n1.A
  ) {desc=Migrated from FeatureLevelSwitch}
 ]
}
