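# Unlit translucent sprite-particle material for the echo effect on the floor player jump pad trap.
# Particle color RGB runs through the shared MF_FX_EmissiveAdjustments function to form an emissive color.
# An opacity-style mask multiplies particle alpha by the red channel of the echo texture and a constant-distance DepthFade, softening intersections with scene geometry.
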Material /Game/Weapons/FORT_TRAPS/Floor_Player_Jump_Pad/Effects/M_PlayerJumpPadEcho
graph {
 properties {
  bUsedWithParticleSprites=true
  blendMode=BLEND_Translucent
  pixelDepthOffsetMode=PDOM_Legacy
  shadingModel=MSM_Unlit
  usedShadingModels=Unlit
 }
 nodes [
  n1 = ParticleColor()
  MaterialFunctionCall(Emissive=@n1.RGB) {materialFunction=&MF_FX_EmissiveAdjustments}
  Multiply(
   A=Multiply(A=@n1.A, B=TextureSample() {texture=&/T_PlayerJumpPadEcho}.R)
   B=DepthFade(FadeDistance=Constant() {r=12})
  )
 ]
}
