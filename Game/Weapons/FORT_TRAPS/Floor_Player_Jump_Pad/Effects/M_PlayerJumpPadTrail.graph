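# Additive unlit material for the trail effect of the player jump pad floor trap, usable on beam trails.
# Particle color drives emissive through a shared FX emissive adjustment function.
# A panning noise sample distorts the flipped, scrolling second-UV lookup into the jump pad texture.
# That texture's red channel combines with particle alpha, sharp fades at both U ends of the first UV set, and depth fade to form the opacity mask.
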Material /Game/Weapons/FORT_TRAPS/Floor_Player_Jump_Pad/Effects/M_PlayerJumpPadTrail
graph {
 properties {
  bUsedWithBeamTrails=true
  bUsedWithNanite=true
  blendMode=BLEND_Additive
  pixelDepthOffsetMode=PDOM_Legacy
  shadingModel=MSM_Unlit
  usedShadingModels=Unlit
 }
 nodes [
  n1 = ParticleColor()
  MaterialFunctionCall(Emissive=@n1.RGB) {materialFunction=&MF_FX_EmissiveAdjustments}
  n2 = OneMinus(None=TextureCoordinate() {coordinateIndex=1})
  n3 = TextureSample(UVs=Panner(Coordinate=TextureCoordinate() {coordinateIndex=1}) {speedX=2}) {texture=&T_NoisePack_01}
  n4 = Subtract(A=Multiply(A=MaterialFunctionCall(X=@n3.R, Y=@n3.R) {materialFunction=&MakeFloat2}.Result) {constB=2})
  n5 = TextureSample(UVs=Add(A=Panner(Coordinate=@n2) {speedX=-1}, B=Multiply(A=@n4) {constB=0.05})) {texture=&/T_PlayerJumpPad}
  n6 = ComponentMask(None=TextureCoordinate()) {r=true}
  n7 = Power(Base=Clamp(None=Multiply(A=@n6) {constB=16}))
  n8 = OneMinus(None=ComponentMask(None=TextureCoordinate()) {r=true})
  n9 = Power(Base=Clamp(None=Multiply(A=@n8) {constB=16}))
  Multiply(A=DepthFade(FadeDistance=Constant() {r=16}), B=Multiply(A=Multiply(A=@n1.A, B=@n5.R), B=Multiply(A=@n7, B=@n9)))
 ]
}
