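# Unlit translucent sparkle particle material for the Floor Player Jump Pad trap effects.
# Color is a random per-particle blend of Color A and Color B, multiplied by particle color and passed through MF_FX_EmissiveAdjustments.
# The mobile shading path skips the blend and uses Color A alone.
# Alpha multiplies particle alpha by the sparkle texture's red channel and its green channel raised to the Exponent parameter.
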Material /Game/Weapons/FORT_TRAPS/Floor_Player_Jump_Pad/Effects/M_Sparkle_01
graph {
 properties {
  bUsedWithNanite=true
  bUsedWithParticleSprites=true
  blendMode=BLEND_Translucent
  pixelDepthOffsetMode=PDOM_Legacy
  shadingModel=MSM_Unlit
  usedShadingModels=Unlit
 }
 nodes [
  Constant()
  n1 = ParticleColor()
  n2 = VectorParameter() {parameterName=Color A, defaultValue=<rgba 1 0.5534 0.295 0>}
  n3 = LinearInterpolate(A=@n2.RGB, B=VectorParameter() {parameterName=Color B, defaultValue=<rgba 0.401 0.3782 1 0>}.RGB, Alpha=ParticleRandom())
  MaterialFunctionCall(Emissive=Multiply(A=@n1.RGB, B=ShadingPathSwitch(Default=@n3, Mobile=@n2.RGB) {desc=Migrated from FeatureLevelSwitch})) {materialFunction=&MF_FX_EmissiveAdjustments}
  n4 = TextureSample() {texture=/Game/Effects/Fort_Effects/Textures/Sparkle/T_Sparkle_01}
  n5 = ScalarParameter() {defaultValue=1, parameterName=Exponent}
  Multiply(A=@n1.A, B=Multiply(A=@n4.R, B=Power(Base=@n4.G, Exp=@n5)))
 ]
}
