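# Unlit, translucent, two-sided material for mesh particles, named for the upward jump effect of the floor player jump pad trap.
# Its only logic passes particle color RGB into the emissive input of the shared MF_FX_EmissiveAdjustments function.
# The function's result is not shown connected to any material output, and no opacity input is set.
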Material /Game/Weapons/FORT_TRAPS/Floor_Player_Jump_Pad/M_Player_Jump_UP_Material
graph {
 properties {
  bUsedWithMeshParticles=true
  blendMode=BLEND_Translucent
  pixelDepthOffsetMode=PDOM_Legacy
  shadingModel=MSM_Unlit
  twoSided=true
  usedShadingModels=Unlit
 }
 nodes [MaterialFunctionCall(Emissive=ParticleColor().RGB) {materialFunction=&MF_FX_EmissiveAdjustments}]
}
